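; Apollo road proxy actor, a stand-in road meant to appear only in PC performance mode.
; On BeginPlay it waits for playlist data, then checks its UseRoadProxyInPerfMode flag and the platform's perf-mode state to log it as visible.
; The removal path sets a short lifespan, destroys the actor and logs the destruction.
; Tick, overlap and construction script are empty.
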(event EventBeginPlay
  (Class|FortAsyncActionCurrentPlaylistReady|PlaylistDataReadyAsync
    (:OnReady
      (if (Variables|Default|GetUseRoadProxyInPerfMode)
        (if (Class|PlatformHelpersLibrary|IsPcPerfMode)
          (Fortnite|FortShippingLog self "Apollo Road Proxy is visible")
          (else
            _)))
      (Actor|SetLifeSpan 0.2)
      (Actor|DestroyActor)
      (Fortnite|FortShippingLog self "Apollo Road Proxy was destroyed"))))

(event Collision|EventActorBeginOverlap
  (params (OtherActor :type Actor%20Object%20Reference)))

(event EventTick
  (params (DeltaSeconds :type Float%20%28single-precision%29)))

(fn ConstructionScript)
